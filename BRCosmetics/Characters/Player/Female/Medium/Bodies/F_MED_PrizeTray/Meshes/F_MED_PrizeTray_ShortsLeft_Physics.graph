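# Physics asset for the left shorts leg of the F_MED_PrizeTray female player cosmetic body, set up for secondary cloth-like motion through bones.
# Four two-link capsule chains (front, back, inner, outer) hang from small base bodies, which connect to the thigh with free rotation.
# Chain joints lock twist and allow only a few degrees of swing, looser at the lower link.
# Larger thigh and calf capsules give the leg's volume.
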PhysicsAsset /BRCosmetics/Characters/Player/Female/Medium/Bodies/F_MED_PrizeTray/Meshes/F_MED_PrizeTray_ShortsLeft_Physics
bodies {
 base_shorts_bk_l [{shapeType=Capsule, length=1, radius=1, shapeName=base_shorts_bk_l_capsule, rotation=<rot -9.9617 5.0767 -0.8805>}]
 base_shorts_fr_l [{shapeType=Capsule, length=1, radius=1, shapeName=base_shorts_fr_l_capsule}]
 base_shorts_in_l [{shapeType=Capsule, length=1, radius=1, shapeName=base_shorts_in_l_capsule}]
 base_shorts_out_l [{shapeType=Capsule, length=1, radius=1, shapeName=base_shorts_out_l_capsule}]
 calf_l [{
   shapeType=Capsule
   length=17.9372
   radius=6.256
   shapeName=calf_l_capsule
   center=<-8.0791 -2.6616 -0.0608>
   rotation=<rot -1.3167 100.363 91.317>
  }]
 dyn_shorts_bk_01_l [{shapeType=Capsule, length=10.3197, radius=1.5733, shapeName=dyn_shorts_out_02_l_capsule, center=<1 0 2>, rotation=<rot 0 0 90>}]
 dyn_shorts_bk_02_l [{shapeType=Capsule, length=8.8619, radius=2.7115, shapeName=dyn_shorts_out_02_l_capsule, center=<1 0 1.33>, rotation=<rot 0 0 90>}]
 dyn_shorts_fr_01_l [{shapeType=Capsule, length=10.8197, radius=1.5733, shapeName=dyn_shorts_out_02_l_capsule, center=<1 0 1.4>, rotation=<rot 0 0 90>}]
 dyn_shorts_fr_02_l [{shapeType=Capsule, length=9.1113, radius=2.4621, shapeName=dyn_shorts_out_02_l_capsule, center=<1 0 1.4>, rotation=<rot 0 0 90>}]
 dyn_shorts_in_01_l [{shapeType=Capsule, length=10.8197, radius=1.5733, shapeName=dyn_shorts_out_02_l_capsule, center=<1 0 0>, rotation=<rot 0 -0.0001 100>}]
 dyn_shorts_in_02_l [{shapeType=Capsule, length=9.8712, radius=1.7021, shapeName=dyn_shorts_out_02_l_capsule, center=<1 0 0.45>, rotation=<rot 0 0 90>}]
 dyn_shorts_out_01_l [{shapeType=Capsule, length=10.8197, radius=1.5733, shapeName=dyn_shorts_out_02_l_capsule, center=<1 0 0>, rotation=<rot 0 0 90>}]
 dyn_shorts_out_02_l [{shapeType=Capsule, length=10.0713, radius=1.502, shapeName=dyn_shorts_out_02_l_capsule, center=<1 0 1.4>, rotation=<rot -27.9608 0.0002 90>}]
 thigh_l [{shapeType=Capsule, length=17.6263, radius=7.068, center=<-23.0884 0 0>, rotation=<rot 0 90 90>}]
}
constraints [
 {
  bone1Name=dyn_shorts_fr_01_l
  bone2Name=base_shorts_fr_l
  swing1LimitDegrees=2
  swing2LimitDegrees=5
  twistLimitDegrees=45
  twistMotion=Locked
 }
 {
  bone1Name=dyn_shorts_fr_02_l
  bone2Name=dyn_shorts_fr_01_l
  swing1LimitDegrees=2.5
  swing2LimitDegrees=10
  twistLimitDegrees=45
  twistMotion=Locked
 }
 {
  bone1Name=dyn_shorts_bk_02_l
  bone2Name=dyn_shorts_bk_01_l
  swing1LimitDegrees=2
  swing2LimitDegrees=10
  twistLimitDegrees=45
  twistMotion=Locked
 }
 {
  bone1Name=dyn_shorts_bk_01_l
  bone2Name=base_shorts_bk_l
  swing1LimitDegrees=0
  swing2LimitDegrees=2.5
  twistLimitDegrees=45
  twistMotion=Locked
 }
 {
  bone1Name=dyn_shorts_in_02_l
  bone2Name=dyn_shorts_in_01_l
  swing1LimitDegrees=2
  swing2LimitDegrees=5
  twistLimitDegrees=45
  twistMotion=Locked
 }
 {
  bone1Name=dyn_shorts_in_01_l
  bone2Name=base_shorts_in_l
  swing1LimitDegrees=0
  swing2LimitDegrees=2.5
  twistLimitDegrees=45
  twistMotion=Locked
 }
 {
  bone1Name=dyn_shorts_out_01_l
  bone2Name=base_shorts_out_l
  swing1LimitDegrees=0
  swing2LimitDegrees=5
  twistLimitDegrees=45
  twistMotion=Locked
 }
 {
  bone1Name=dyn_shorts_out_02_l
  bone2Name=dyn_shorts_out_01_l
  swing1LimitDegrees=2.5
  swing2LimitDegrees=10
  twistLimitDegrees=45
  twistMotion=Locked
 }
 {
  bone1Name=base_shorts_bk_l
  bone2Name=thigh_l
  swing1LimitDegrees=45
  swing1Motion=Free
  swing2LimitDegrees=45
  swing2Motion=Free
  twistLimitDegrees=45
  twistMotion=Free
 }
 {
  bone1Name=base_shorts_fr_l
  bone2Name=thigh_l
  swing1LimitDegrees=45
  swing1Motion=Free
  swing2LimitDegrees=45
  swing2Motion=Free
  twistLimitDegrees=45
  twistMotion=Free
 }
 {
  bone1Name=base_shorts_in_l
  bone2Name=thigh_l
  swing1LimitDegrees=45
  swing1Motion=Free
  swing2LimitDegrees=45
  swing2Motion=Free
  twistLimitDegrees=45
  twistMotion=Free
 }
 {
  bone1Name=base_shorts_out_l
  bone2Name=thigh_l
  swing1LimitDegrees=45
  swing1Motion=Free
  swing2LimitDegrees=45
  swing2Motion=Free
  twistLimitDegrees=45
  twistMotion=Free
 }
 {
  bone1Name=calf_l
  bone2Name=thigh_l
  swing1LimitDegrees=45
  swing1Motion=Free
  swing2LimitDegrees=45
  swing2Motion=Free
  twistLimitDegrees=45
  twistMotion=Free
 }
]
