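# Physics asset for the right leg of the shorts on the female medium PrizeTray player body, defining capsule bodies and joints for cloth-like bone motion.
# Front, back, inner and outer chains each link a tiny base anchor to two dynamic capsules, with twist locked and swing limited to a few degrees.
# Larger thigh and calf capsules supply leg volumes, joined by a free-rotating constraint and unconstrained to the shorts chains.
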PhysicsAsset /BRCosmetics/Characters/Player/Female/Medium/Bodies/F_MED_PrizeTray/Meshes/F_MED_PrizeTray_ShortsRight_Physics
bodies {
 base_shorts_bk_r [{shapeType=Capsule, length=0.505, radius=0.505, shapeName=base_shorts_bk_r_capsule, rotation=<rot 0 -90 0>}]
 base_shorts_fr_r [{shapeType=Capsule, length=0.505, radius=0.505, shapeName=base_shorts_fr_r_capsule, rotation=<rot 0 -90 0>}]
 base_shorts_in_r [{shapeType=Capsule, length=0.505, radius=0.505, shapeName=base_shorts_in_r_capsule, rotation=<rot 0 -90 0>}]
 base_shorts_out_r [{shapeType=Capsule, length=0.505, radius=0.505, shapeName=base_shorts_out_r_capsule, rotation=<rot 0 -90 0>}]
 calf_r [{
   shapeType=Capsule
   length=17.9372
   radius=6.256
   shapeName=calf_r_capsule
   center=<8.0791 2.6616 0.0608>
   rotation=<rot -1.3167 100.363 -88.683>
  }]
 dyn_shorts_bk_01_r [{shapeType=Capsule, length=10.8197, radius=1.5733, shapeName=dyn_shorts_out_02_r_capsule, center=<-1 0 -2>, rotation=<rot 0 0 -90>}]
 dyn_shorts_bk_02_r [{shapeType=Capsule, length=8.8619, radius=2.7115, shapeName=dyn_shorts_out_02_r_capsule, center=<-1 0 -2>, rotation=<rot 0 0 -90>}]
 dyn_shorts_fr_01_r [{shapeType=Capsule, length=10.8197, radius=1.5733, shapeName=dyn_shorts_out_02_r_capsule, center=<-1 0 -1.4>, rotation=<rot 0 0 -90>}]
 dyn_shorts_fr_02_r [{shapeType=Capsule, length=9.1113, radius=2.4621, shapeName=dyn_shorts_out_02_r_capsule, center=<-1 0 -1.4>, rotation=<rot 0 0 -90>}]
 dyn_shorts_in_01_r [{shapeType=Capsule, length=10.8197, radius=1.5733, shapeName=dyn_shorts_out_02_r_capsule, center=<-1 0 0>, rotation=<rot 0 -0.0001 -80>}]
 dyn_shorts_in_02_r [{shapeType=Capsule, length=9.8712, radius=1.7021, shapeName=dyn_shorts_out_02_r_capsule, center=<-1 0 -0.45>, rotation=<rot 0 0 -90>}]
 dyn_shorts_out_01_r [{shapeType=Capsule, length=10.8197, radius=1.5733, shapeName=dyn_shorts_out_02_r_capsule, center=<-1 0 0>, rotation=<rot 0 0 -90>}]
 dyn_shorts_out_02_r [{shapeType=Capsule, length=10.0713, radius=1.502, shapeName=dyn_shorts_out_02_r_capsule, center=<-1 0 -1.4>, rotation=<rot 0 0 -90>}]
 thigh_r [{shapeType=Capsule, length=18.1263, radius=7.068, center=<23.0884 0 0>, rotation=<rot 0 90 -90>}]
}
constraints [
 {
  bone1Name=dyn_shorts_out_01_r
  bone2Name=base_shorts_out_r
  swing1LimitDegrees=0
  swing2LimitDegrees=5
  twistLimitDegrees=45
  twistMotion=Locked
 }
 {
  bone1Name=dyn_shorts_out_02_r
  bone2Name=dyn_shorts_out_01_r
  swing1LimitDegrees=2.5
  swing2LimitDegrees=10
  twistLimitDegrees=45
  twistMotion=Locked
 }
 {
  bone1Name=dyn_shorts_bk_01_r
  bone2Name=base_shorts_bk_r
  swing1LimitDegrees=0
  swing2LimitDegrees=2.5
  twistLimitDegrees=45
  twistMotion=Locked
 }
 {
  bone1Name=dyn_shorts_bk_02_r
  bone2Name=dyn_shorts_bk_01_r
  swing1LimitDegrees=2
  swing2LimitDegrees=10
  twistLimitDegrees=45
  twistMotion=Locked
 }
 {
  bone1Name=dyn_shorts_fr_01_r
  bone2Name=base_shorts_fr_r
  swing1LimitDegrees=2
  swing2LimitDegrees=5
  twistLimitDegrees=45
  twistMotion=Locked
 }
 {
  bone1Name=dyn_shorts_fr_02_r
  bone2Name=dyn_shorts_fr_01_r
  swing1LimitDegrees=2.5
  swing2LimitDegrees=10
  twistLimitDegrees=45
  twistMotion=Locked
 }
 {
  bone1Name=dyn_shorts_in_01_r
  bone2Name=base_shorts_in_r
  swing1LimitDegrees=0
  swing2LimitDegrees=2.5
  twistLimitDegrees=45
  twistMotion=Locked
 }
 {
  bone1Name=dyn_shorts_in_02_r
  bone2Name=dyn_shorts_in_01_r
  swing1LimitDegrees=2
  swing2LimitDegrees=5
  twistLimitDegrees=45
  twistMotion=Locked
 }
 {
  bone1Name=calf_r
  bone2Name=thigh_r
  swing1LimitDegrees=45
  swing1Motion=Free
  swing2LimitDegrees=45
  swing2Motion=Free
  twistLimitDegrees=45
  twistMotion=Free
 }
]
