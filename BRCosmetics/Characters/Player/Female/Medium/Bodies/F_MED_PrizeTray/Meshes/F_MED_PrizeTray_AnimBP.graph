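# Post-process style AnimBP for the F_MED_PrizeTray cosmetic body, adding shorts motion to the pose copied from its attached parent mesh.
# A Control Rig receives a crouching-not-sliding value, then per-thigh RigidBody nodes simulate each leg, followed by fixed additive shorts-bone offsets.
# A front-end idle curve blends DefaultSlot onto inner-shorts bone branches.
# The update event also stores gravity and rigid-body settings from the male PrizeTray parameter asset.
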AnimBlueprint /BRCosmetics/Characters/Player/Female/Medium/Bodies/F_MED_PrizeTray/Meshes/F_MED_PrizeTray_AnimBP
target_skeleton=&/F_MED_PrizeTray_Skeleton
graphs [
 AnimGraph {nodes [
   n1 = Layered blend per bone(
    BasePose=Use cached pose 'ShortsCache_1'()
    BlendPoses_0=Slot 'DefaultSlot'()
    BlendWeights_0=Get Custom Idle Is Playing Float() -> (Custom Idle Is Playing Float:double)
   ) {layerSetup [{branchFilters [{boneName=dyn_shorts_in_01_l}, {boneName=dyn_shorts_in_02_l}, {boneName=dyn_shorts_in_01_r}, {boneName=dyn_shorts_in_02_r}]}]}
   Output Pose(Result=@n1) {layerGroup=None}
   n2 = Control Rig!(Source=Copy Pose From Mesh() {bUseAttachedParent=true}, Alpha, OnlyCrouching:double=Get OnlyCrouching() -> (OnlyCrouching:double)) -> (Pose) {controlRigAssetReference {blueprintRigClass=&/F_MED_PrizeTray_CtrlRig.F_MED_PrizeTray_CtrlRig_C}}
   Save cached pose 'ControlRigCache'(Pose=@n2) {cacheName=ControlRigCache}
   n3 = RigidBody!(
    ComponentPose=Local To Component(LocalPose=Use cached pose 'ControlRigCache'())
    Alpha
    OverrideWorldGravity:Vector="0,0,-980"
    ExternalForce
    SimSpaceSettings:Sim Space Settings Structure=(WorldAlpha=0,VelocityScaleZ=1,DampingAlpha=1,MaxLinearVelocity=10000,MaxAngularVelocity=10000,MaxLinearAcceleration=10000,MaxAngularAcceleration=10000,ExternalLinearDragV=(X=0,Y=0,Z=0),ExternalLinearVelocity=(X=0,Y=0,Z=0),ExternalAngularVelocity=(X=0,Y=0,Z=0))
   ) -> (Pose) {
    bOverrideWorldGravity=true
    overridePhysicsAsset=&/F_MED_PrizeTray_ShortsLeft_Physics
    simulationSpace=BaseBoneSpace
    baseBoneRef {boneName=thigh_l}
    componentAppliedLinearAccClamp=<0 0 0>
    overrideWorldGravity {z=-980}
   }
   n4 = RigidBody(
    ComponentPose=@n3
    OverrideWorldGravity="0,0,-980"
    SimSpaceSettings=(WorldAlpha=0,VelocityScaleZ=1,DampingAlpha=1,MaxLinearVelocity=10000,MaxAngularVelocity=10000,MaxLinearAcceleration=10000,MaxAngularAcceleration=10000,ExternalLinearDragV=(X=0,Y=0,Z=0),ExternalLinearVelocity=(X=0,Y=0,Z=0),ExternalAngularVelocity=(X=0,Y=0,Z=0))
   ) {
    bOverrideWorldGravity=true
    overridePhysicsAsset=&/F_MED_PrizeTray_ShortsRight_Physics
    simulationSpace=BaseBoneSpace
    baseBoneRef {boneName=thigh_r}
    componentAppliedLinearAccClamp=<0 0 0>
    overrideWorldGravity {z=-980}
   }
   Save cached pose 'DynamicsCache'(Pose=Component To Local(ComponentPose=@n4)) {cacheName=DynamicsCache}
   n5 = "Transform (Modify) Bone - Bone: base_shorts_fr_r"!(
    ComponentPose=Local To Component(LocalPose=Use cached pose 'DynamicsCache'())
    bAlphaBoolEnabled:bool=True
    Translation="0,0,0"
    Rotation="3,0,0"
    Scale="1,1,1"
   ) -> (Pose) {
    alphaInputType=Bool
    rotationMode=BMM_Additive
    rotationSpace=BCS_BoneSpace
    translationMode=BMM_Additive
    translationSpace=BCS_BoneSpace
    alphaBoolBlend {blendInTime=0.2, blendOutTime=0.2}
    boneToModify {boneName=base_shorts_fr_r}
    rotation {pitch=3}
   }
   n6 = "Transform (Modify) Bone - Bone: base_shorts_out_r"!(ComponentPose=@n5, bAlphaBoolEnabled:bool=True, Translation="-10,0,0", Rotation="0,0,0", Scale="1,1,1") -> (Pose) {
    alphaInputType=Bool
    rotationMode=BMM_Additive
    rotationSpace=BCS_BoneSpace
    translationMode=BMM_Additive
    translationSpace=BCS_BoneSpace
    alphaBoolBlend {blendInTime=0.2, blendOutTime=0.2}
    boneToModify {boneName=base_shorts_out_r}
    translation {x=-10}
   }
   n7 = "Transform (Modify) Bone - Bone: base_shorts_bk_r"!(ComponentPose=@n6, bAlphaBoolEnabled:bool=True, Translation="-5,0,0", Rotation="-45,0,0", Scale="1,1,1") -> (Pose) {
    alphaInputType=Bool
    rotationMode=BMM_Additive
    rotationSpace=BCS_BoneSpace
    translationMode=BMM_Additive
    translationSpace=BCS_BoneSpace
    alphaBoolBlend {blendInTime=0.2, blendOutTime=0.2}
    boneToModify {boneName=base_shorts_bk_r}
    rotation {pitch=-45}
    translation {x=-5}
   }
   n8 = "Transform (Modify) Bone - Bone: base_shorts_fr_l"!(ComponentPose=@n7, bAlphaBoolEnabled:bool=True, Translation="0,0,0", Rotation="4,0,0", Scale="1,1,1") -> (Pose) {
    alphaInputType=Bool
    rotationMode=BMM_Additive
    rotationSpace=BCS_BoneSpace
    translationMode=BMM_Additive
    translationSpace=BCS_BoneSpace
    alphaBoolBlend {blendInTime=0.2, blendOutTime=0.2}
    boneToModify {boneName=base_shorts_fr_l}
    rotation {pitch=4}
   }
   n9 = "Transform (Modify) Bone - Bone: base_shorts_out_l"!(ComponentPose=@n8, bAlphaBoolEnabled:bool=True, Translation="10,0,0", Rotation="0,0,0", Scale="1,1,1") -> (Pose) {
    alphaInputType=Bool
    rotationMode=BMM_Additive
    rotationSpace=BCS_BoneSpace
    translationMode=BMM_Additive
    translationSpace=BCS_BoneSpace
    alphaBoolBlend {blendInTime=0.2, blendOutTime=0.2}
    boneToModify {boneName=base_shorts_out_l}
    translation {x=10}
   }
   n10 = "Transform (Modify) Bone - Bone: base_shorts_bk_l"!(ComponentPose=@n9, bAlphaBoolEnabled:bool=True, Translation="5,0,0", Rotation="-45,0,0", Scale="1,1,1") -> (Pose) {
    alphaInputType=Bool
    rotationMode=BMM_Additive
    rotationSpace=BCS_BoneSpace
    translationMode=BMM_Additive
    translationSpace=BCS_BoneSpace
    alphaBoolBlend {blendInTime=0.2, blendOutTime=0.2}
    boneToModify {boneName=base_shorts_bk_l}
    rotation {pitch=-45}
    translation {x=5}
   }
   Save cached pose 'ShortsCache_1'(Pose=Component To Local(ComponentPose=@n10)) {cacheName=ShortsCache_1}
  ]}
 EventGraph {nodes [
   Event BlueprintUpdateAnimation()
   > n1 = Sequence!(execute) -> (then_0, then_1, then_2:Exec)
   > n2 = "Gravity Override State Function (Parameter Asset) (Nativized)"(
    CharacterPartAnimInstance=Self-Reference()
    GameTimeInSeconds=GetGameTimeInSeconds()
    Parameters=&/M_MED_PrizeTray_Parameters
   )
   > Set Gravity Override(Gravity Override=@n2)
   > Set RBAN Settings Out(RBAN Settings Out=@n2.OutRigidBodyAnimNodeParameters)
   Branch(execute=@n1.then_2, Condition=Get bIsInFrontEnd())
   > Set Custom Idle Is Playing Float(Custom Idle Is Playing Float:double=GetCurveValue(self=GetOwnerFrontEndAnimInstance(), CurveName=Frontend_Idle_Bypass)) -> (Output_Get:double)
   n3 = NOT Boolean(A=Get bIsSliding())
   n4 = Get bIsCrouching()
   Set OnlyCrouching(execute=@n1.then_1, OnlyCrouching:double="To Float (Boolean)"(InBool=AND Boolean(A=@n4, B=@n3))) -> (Output_Get:double)
  ]}
]
variables [
 Custom Idle Is Playing Float {type=number}
 Gravity Override {type=Vector}
 OnlyCrouching {type=number}
 RBAN Settings Out {type=FortRigidBodyAnimNodeParameters}
]
