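# Physics asset for the F_MED_PrizeTray female medium player body cosmetic, giving its skeleton a simplified set of collision bodies.
# Single capsules cover the pelvis, two upper spine bones, upper arms, hands, thighs and calves, with no head, forearm or foot bodies.
# Constraints chain these outward from the pelvis, linking hands directly to upper arms.
# Both hip constraints explicitly free swing and twist.
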PhysicsAsset /BRCosmetics/Characters/Player/Female/Medium/Bodies/F_MED_PrizeTray/Meshes/F_MED_PrizeTray_Physics
bodies {
 calf_l [{
   shapeType=Capsule
   length=17.9372
   radius=6.256
   shapeName=calf_l_capsule
   center=<-8.0791 -2.6616 -0.0608>
   rotation=<rot -1.3167 100.363 91.317>
  }]
 calf_r [{
   shapeType=Capsule
   length=17.9372
   radius=6.256
   shapeName=calf_r_capsule
   center=<8.0791 2.6616 0.0608>
   rotation=<rot -1.3167 100.363 -88.683>
  }]
 hand_l [{shapeType=Capsule, length=9.0541, radius=5.4211, center=<7.8534 -2.5233 -0.3794>, rotation=<rot 0 80.3714 -79.3821>}]
 hand_r [{shapeType=Capsule, length=9.0541, radius=5.4211, center=<-7.8534 2.5233 0.3794>, rotation=<rot 0 80.3714 -79.3821>}]
 pelvis [{shapeType=Capsule, length=10.6313, radius=8.4504, center=<-2.4549 -2.3567 0.2226>}]
 spine_04 [{shapeType=Capsule, length=6.739, radius=8.8524, center=<-11.9664 0.5516 -0.1029>, rotation=<rot -1.171 -90 -0.454>}]
 spine_05 [{shapeType=Capsule, length=14.1027, radius=10.2719, center=<-1.6687 -0.8653 -0.0102>, rotation=<rot 0.0374 -90 -0.0537>}]
 thigh_l [{shapeType=Capsule, length=17.6263, radius=7.068, center=<-23.0884 0 0>, rotation=<rot 0 90 90>}]
 thigh_r [{shapeType=Capsule, length=18.1263, radius=7.068, center=<23.0884 0 0>, rotation=<rot 0 90 -90>}]
 upperarm_l [{shapeType=Capsule, length=25.2146, radius=7.8858, center=<23.0636 2.8082 -1.513>, rotation=<rot 0 100.7581 -87.3147>}]
 upperarm_r [{shapeType=Capsule, length=25.2137, radius=7.885, center=<-23.0657 -2.808 1.4948>, rotation=<rot 0 100.7539 -87.2623>}]
}
constraints [
 {bone1Name=spine_04, bone2Name=pelvis, swing1LimitDegrees=45, swing2LimitDegrees=45, twistLimitDegrees=45}
 {bone1Name=spine_05, bone2Name=spine_04, swing1LimitDegrees=45, swing2LimitDegrees=45, twistLimitDegrees=45}
 {bone1Name=upperarm_l, bone2Name=spine_05, swing1LimitDegrees=45, swing2LimitDegrees=45, twistLimitDegrees=45}
 {bone1Name=hand_l, bone2Name=upperarm_l, swing1LimitDegrees=45, swing2LimitDegrees=45, twistLimitDegrees=45}
 {bone1Name=upperarm_r, bone2Name=spine_05, swing1LimitDegrees=45, swing2LimitDegrees=45, twistLimitDegrees=45}
 {bone1Name=hand_r, bone2Name=upperarm_r, swing1LimitDegrees=45, swing2LimitDegrees=45, twistLimitDegrees=45}
 {
  bone1Name=thigh_l
  bone2Name=pelvis
  swing1LimitDegrees=45
  swing1Motion=Free
  swing2LimitDegrees=45
  swing2Motion=Free
  twistLimitDegrees=45
  twistMotion=Free
 }
 {bone1Name=calf_l, bone2Name=thigh_l, swing1LimitDegrees=45, swing2LimitDegrees=45, twistLimitDegrees=45}
 {bone1Name=calf_r, bone2Name=thigh_r, swing1LimitDegrees=45, swing2LimitDegrees=45, twistLimitDegrees=45}
 {
  bone1Name=thigh_r
  bone2Name=pelvis
  swing1LimitDegrees=45
  swing1Motion=Free
  swing2LimitDegrees=45
  swing2Motion=Free
  twistLimitDegrees=45
  twistMotion=Free
 }
]
